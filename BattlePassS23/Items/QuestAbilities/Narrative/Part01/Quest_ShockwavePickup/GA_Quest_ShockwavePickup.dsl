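; Battle Pass S23 quest ability that reports to the owning player's quest manager when they already carry a Shockwave Mace.
; On activation it checks for a Battle Royale playlist and the required game phase, then runs the inventory check.
; The check loops over inventory items matching the mace's gameplay tag to set a found flag.
; If the flag is set, it waits briefly and sends a custom stat event tagged as already having the item.
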(event Ability|EventActivateAbility
  ;[0 0 0] Check if BR match is in progress when ability activates
  (|CallParentFunction)
  (bind _bvalidplaylist (CallFunction|CheckifValidPlaylist))
  (if _bvalidplaylist
    (bind _bisingameplayphase (CallFunction|CheckGamePhase))
    (if _bisingameplayphase
      (CallFunction|CheckInventoryforItem)
      (else
        _))
    (else
      _))
  (Ability|EndAbility))

(event Custom|CheckInventoryForItem
  ;[0 0 0] Check if the player already has this item in their inventory; if not, end the ability
  (bind _items (InventoryManagement|GetItemsfromInventoryOwner (Fortnite|GetFortPlayerControllerfromActor (Variables|GAVars|Data|GetOwningPlayerPawn))))
  (if (Utilities|Array|IsNotEmpty _items)
    (bind _array_element (Utilities|Array|ForEachLoopwithBreak _items)
      (:LoopBody
        (if (GameplayTags|HasMatchingGameplayTag (Item|GetItemDefinition _array_element) "(TagName=\"Item.Weapon.Melee.ShockwaveMace\")")
          (Development|PrintString "Item found!" true true "(R=1.000000,G=0.000000,B=0.216887,A=1.000000)" 3.0)
          (Variables|Default|SetbItemFound true))
        (Variables|Default|SetbItemFound false))
      (:Completed
        _))
    (if (Variables|Default|GetbItemFound)
      (Utilities|FlowControl|Delay 2.0)
      (Quest|SendComplexCustomStatEvent (Variables|GAVars|Data|GetOwningPlayerQuestManager) 0 (Variables|Default|GetAlreadyHasItemTag) "(GameplayTags=)" 1)
      (else
        _))
    (Ability|EndAbility)
    (else
      _)))

(event Ability|EventOnEndAbility
  (params (bWasCancelled :type Boolean))
  (|CallParentFunction bWasCancelled))

(event Ability|EventFailedToActivatePassiveAbility
  (params (ActorInfo :type Gameplay%20Ability%20Actor%20Info%20Structure))
  (|CallParentFunction ActorInfo))

(fn CheckGamePhase
  ;[0 0 0] Check if the player is in the gameplay phase
  (return (Utilities|Enum|Equal(Enum) (Variables|Default|GetRequiredGamePhase) (BattleRoyaleGamePhaseLogic|GetGamePhase (BattleRoyaleGamePhaseLogic|GetBattleRoyaleGamePhaseLogicComponent)))))

(fn CheckIfValidPlaylist
  ;[0 0 0] Is player in BR
  (return (Class|FortGameStateAthena|IsBattleRoyalePlaylist (Fortnite|Athena|GameState|GetGameStateAthenafromContext (Variables|GAVars|Data|GetOwningPlayerPawn)) false)))
